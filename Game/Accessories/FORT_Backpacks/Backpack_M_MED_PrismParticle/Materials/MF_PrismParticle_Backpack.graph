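# Material function for the PrismParticle backpack, layering a glowing, sparkling, subsurface crystal look onto incoming material attributes.
# A vertex animation texture drives a looping morph through three shapes and back, with RGB supplying pivot-basis normals and alpha scaling world position offset toward a pivot.
# Emissive combines view-aligned reflection, inverted fresnel, sparkle and a spherical glow gradient.
# A Blueprint-timestamped curve pulse adds reactive glow with time-of-day brightness.
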MaterialFunction /Game/Accessories/FORT_Backpacks/Backpack_M_MED_PrismParticle/Materials/MF_PrismParticle_Backpack
graph {nodes [
  Constant3Vector() {constant=<rgba 0 0 1 0>}
  n1 = FunctionInput() {inputType=FunctionInput_MaterialAttributes}
  n2 = DotProduct(A=PixelNormalWS(), B=CameraVectorWS())
  n3 = Power(
   Base=OneMinus(None=Clamp(None=@n2))
   Exp=ScalarParameter() {defaultValue=0.7493, group=PrismParticle, parameterName=PP_SubSurfaceFresnelExponent}
  )
  n4 = ScalarParameter() {defaultValue=0.1035, group=PrismParticle, parameterName=PP_SubSurfaceOpacity}
  n5 = VectorParameter() {group=PrismParticle, parameterName=PP_SubSurfaceColor, defaultValue=<rgba 0.0067 0.0103 0.02 1>}
  n6 = SetMaterialAttributes!(
   MaterialAttributes=@n1
   BaseColor=NamedRerouteUsage()
   Metallic=ScalarParameter() {group=PrismParticle, parameterName=PP_Metallic}
   Specular=ScalarParameter() {defaultValue=0.1514, group=PrismParticle, parameterName=PP_Specular}
   Roughness=ScalarParameter() {defaultValue=0.1101, group=PrismParticle, parameterName=PP_Roughness}
   EmissiveColor=NamedRerouteUsage()
   Normal=NamedRerouteUsage()
   WorldPositionOffset=NamedRerouteUsage()
   ShadingModel=ShadingModel() {shadingModel=MSM_Subsurface}
   Opacity=OneMinus(None=Multiply(A=@n3, B=@n4))
   SubsurfaceColor=@n5.RGB
  ) {attributeSetTypes [
    #1
    #2
    #3
    #4
    #5
    #6
    #7
    #8
    #9
    #10
   ]}
  FunctionOutput(None=@n6)
  n7 = TextureSample!(
   UVs=MaterialFunctionCall!(ReflectionVector, ReflectionTexture) {materialFunction=/Engine/Functions/Engine_MaterialFunctions01/Reflections/ViewAlignedReflection}.UVs
   Tex
   Bias=ScalarParameter() {group=PrismParticle, parameterName=PP_ReflectionMipBias}
   Apply View MipBias
  ) {mipValueMode=TMVM_MipBias, texture=/Game/Packages/Fortress_SharedMaterials/Textures/T_Ravens_SpecularReflection}
  n8 = Multiply(
   A=ScalarParameter() {defaultValue=4, group=PrismParticle, parameterName=PP_InvertedFresnelBrightness}
   B=MaterialFunctionCall(
    Normal Vector=PixelNormalWS()
    Invert Fresnel=StaticBool() {value=true}
    Power=ScalarParameter() {defaultValue=2, group=PrismParticle, parameterName=PP_InvertedFresnelPower}
   ) {materialFunction=&Fresnel_Function}.Result
  )
  n9 = ScalarParameter() {defaultValue=1, group=PrismParticle, parameterName=PP_UseObjectPosition}
  n10 = MaterialFunctionCall!(
   Scale=ScalarParameter() {defaultValue=-0.1, group=PrismParticle, parameterName=PP_SparkleScale}
   Sparkle World center point=LinearInterpolate(A=NamedRerouteUsage(), B=ObjectPositionWS(), Alpha=NamedRerouteDeclaration(None=@n9) {name=UseObjectPosition, nodeColor=<rgba 0.1484 0 1 1>})
  ) {materialFunction=/Game/Packages/Fortnite_UniqueMaterials_1/Effects/FortniteSparkle}
  n11 = Multiply(A=@n10.Result, B=ScalarParameter() {defaultValue=0.211, group=PrismParticle, parameterName=PP_SparkleBrightness})
  n12 = NamedRerouteUsage()
  n13 = Clamp(None=@n12)
  n14 = ScalarParameter() {defaultValue=5.6, group=PrismParticle, parameterName=PP_GlowAmount_Reactive_Day}
  n15 = ScalarParameter() {defaultValue=5.2, group=PrismParticle, parameterName=PP_GlowAmount_Reactive_Night}
  n16 = LinearInterpolate(
   A=ScalarParameter() {defaultValue=2, group=PrismParticle, parameterName=PP_GlowAmount}
   B=MaterialFunctionCall(Color_Morning=@n14, Color_Day=@n14, Color_Evening=@n15, Color_Night=@n15) {materialFunction=&Color_Time_of_Day}.Result
   Alpha=NamedRerouteUsage()
  )
  n17 = Multiply(A=Add(A=Multiply(A=@n7.RGB, B=@n8), B=Add(B=@n11) {constA=0.5}), B=Add(A=Multiply(A=@n13, B=@n16), B=@n12) {constB=0.1})
  n18 = VectorParameter() {group=PrismParticle, parameterName=PP_2DGlowColor, defaultValue=<rgba 0.1537 0.2538 0.53 2>}
  n19 = Add(
   A=Power(Base=Multiply(A=@n17, B=@n18.RGB), Exp=@n18.A)
   B=VectorParameter() {group=PrismParticle, parameterName=PP_GlowPass, defaultValue=<rgba 0 0.0314 0.27 0.1>}.RGB
  )
  LinearInterpolate(
   A=@n19
   B=VectorParameter() {group=PrismParticle, parameterName=PP_PulseColor, defaultValue=<rgba 0.5523 5.1329 20 1>}.RGB
   Alpha=NamedRerouteUsage()
  )
  Multiply(
   A=@n5.RGB
   B=MaterialFunctionCall() {materialFunction=/Game/Packages/Fortress_SharedMaterials/Base_Material_Functions/CascadingShadowMapMask}.Cascading Shadow Mask
  )
  NamedRerouteDeclaration(None=@n19) {name=Emissive_OUT, nodeColor=<rgba 1 0.5156 0 1>}
  n20 = NamedRerouteUsage()
  n21 = AppendVector(
   A=MaterialFunctionCall!(VectorToTransform=Transform(None=@n20), BasisX=NamedRerouteUsage(), BasisY=NamedRerouteUsage(), BasisZ=NamedRerouteUsage()) {materialFunction=/Engine/Functions/Engine_MaterialFunctions02/Math/InverseTransformMatrix}.Transformed Vector 3
   B=Constant()
  )
  n22 = MaterialFunctionCall(Float2=TextureCoordinate()) {materialFunction=&BreakOutFloat2Components}
  n23 = ScalarParameter() {defaultValue=78, group=PrismParticle, parameterName=PP_VAT_TexSize}
  n24 = Divide(B=@n23) {constA=0}
  n25 = TextureObjectParameter() {
   group=PrismParticle
   parameterName=PP_VAT
   samplerType=SAMPLERTYPE_Data
   texture=/Game/Accessories/FORT_Backpacks/Backpack_M_MED_PrismParticle/Textures/T_PrismParticle_Backpack_VAT
  }
  n26 = TextureSample!(
   UVs=MaterialFunctionCall(X=@n22.R, Y=Add(A=@n22.G, B=@n24)) {materialFunction=&MakeFloat2}.Result
   Tex=@n25
   Apply View MipBias
  ) {
   samplerSource=SSM_Wrap_WorldGroupSettings
   samplerType=SAMPLERTYPE_Data
   texture=&T_PrismParticle_Backpack_VAT
  }
  n27 = ScalarParameter() {defaultValue=0.5, group=PrismParticle, parameterName=PP_ShapeDuration}
  n28 = Time()
  n29 = ScalarParameter() {defaultValue=0.15, group=PrismParticle, parameterName=PP_MorphSpeed}
  n30 = Multiply(A=Frac(None=Multiply(A=@n28, B=@n29) {constB=0.2}), B=ScalarParameter() {defaultValue=4, group=PrismParticle, parameterName=PP_ShapesNum}) {constB=3}
  n31 = LinearInterpolate(A=@n21, B=@n26.RGBA, Alpha=SmoothStep(Min=@n27, Value=Clamp(None=@n30)) {constMin=1})
  n32 = Divide(B=@n23)
  n33 = TextureSample(
   UVs=MaterialFunctionCall(X=@n22.R, Y=Add(A=@n22.G, B=@n32)) {materialFunction=&MakeFloat2}.Result
   Tex=@n25
  ) {
   samplerSource=SSM_Wrap_WorldGroupSettings
   samplerType=SAMPLERTYPE_Data
   texture=&T_PrismParticle_Backpack_VAT
  }
  n34 = Subtract(A=@n30)
  n35 = SmoothStep(Min=@n27, Value=Clamp(None=@n34))
  n36 = Divide(B=@n23) {constA=2}
  n37 = TextureSample(
   UVs=MaterialFunctionCall(X=@n22.R, Y=Add(A=@n22.G, B=@n36)) {materialFunction=&MakeFloat2}.Result
   Tex=@n25
  ) {
   desc=platonic
   samplerSource=SSM_Wrap_WorldGroupSettings
   samplerType=SAMPLERTYPE_Data
   texture=&T_PrismParticle_Backpack_VAT
  }
  n38 = Subtract(A=@n34)
  n39 = Clamp(None=@n38)
  n40 = Subtract(A=@n38)
  n41 = LinearInterpolate(
   A=LinearInterpolate(A=LinearInterpolate(A=@n31, B=@n33.RGBA, Alpha=@n35), B=@n37.RGBA, Alpha=SmoothStep(Min=@n27, Value=@n39))
   B=@n21
   Alpha=SmoothStep(Min=@n27, Value=Clamp(None=@n40))
  )
  n42 = Normalize(VectorInput=ComponentMask(None=@n41) {b=true, g=true, r=true})
  n43 = VectorParameter() {group=PrismParticle, parameterName=PP_PivotBasisX, defaultValue=<rgba 1 0 0 0>}
  n44 = VectorParameter() {group=PrismParticle, parameterName=PP_PivotBasisY, defaultValue=<rgba 0 1 0 0>}
  n45 = VectorParameter() {group=PrismParticle, parameterName=PP_PivotBasisZ, defaultValue=<rgba 0 0 1 0>}
  NamedRerouteDeclaration(None=MaterialFunctionCall(
    VectorToTransform=VertexInterpolator(VS=@n42).PS
    BasisX=NamedRerouteDeclaration(None=@n43.RGB) {name=BasisX, nodeColor=<rgba 1 0 0.2578 1>}
    BasisY=NamedRerouteDeclaration(None=@n44.RGB) {name=BasisY, nodeColor=<rgba 0.9922 1 0 1>}
    BasisZ=NamedRerouteDeclaration(None=@n45.RGB) {name=BasisZ, nodeColor=<rgba 0.9453 0 1 1>}
   ) {materialFunction=&Transform3x3Matrix}.Transformed Vector) {name=Normal_OUT, nodeColor=<rgba 0 0.6484 1 1>}
  NamedRerouteDeclaration(None=GetMaterialAttributes(None=@n1) {attributeGetTypes [#6]}.Normal) {name=Normal_IN, nodeColor=<rgba 0 0.3437 1 1>}
  n46 = ComponentMask(None=@n41) {a=true}
  n47 = VectorParameter() {group=PrismParticle, parameterName=PP_PivotLocation, defaultValue=<rgba 0 -12 0 1>}
  n48 = Subtract(
   A=LinearInterpolate(
    A=NamedRerouteDeclaration(None=@n47.RGB) {name=PivotLocation, nodeColor=<rgba 0 1 0.1094 1>}
    B=ObjectPositionWS()
    Alpha=NamedRerouteUsage()
   )
   B=WorldPosition().XYZ
  )
  NamedRerouteDeclaration(None=Multiply(A=Multiply(A=@n46, B=@n48), B=ScalarParameter() {defaultValue=1, group=PrismParticle, parameterName=PP_WPO_Intensity})) {name=WPO_OUT, nodeColor=<rgba 0.4297 0 1 1>}
  n49 = VectorParameter() {group=PrismParticle, parameterName=PP_DiffuseFresnelColor, defaultValue=<rgba 0.0152 0.0927 0.415 1>}
  n50 = Fresnel() {exponent=3}
  n51 = VectorParameter() {group=PrismParticle, parameterName=PP_ColorInterior, defaultValue=<rgba 0 0.061 0.325 1>}
  n52 = VectorParameter() {group=PrismParticle, parameterName=PP_Color, defaultValue=<rgba 0.1159 0.2286 1 1>}
  n53 = Fresnel() {exponent=2}
  NamedRerouteDeclaration(None=Add(A=Multiply(A=@n49.RGB, B=@n50), B=LinearInterpolate(A=@n51.RGB, B=@n52.RGB, Alpha=@n53))) {name=BaseColor_OUT, nodeColor=<rgba 0.6641 1 0 1>}
  NamedRerouteDeclaration() {name=Debug, nodeColor=<rgba 0.0547 0 1 1>}
  n54 = CurveAtlasRowParameter(CurveTime=Frac(None=NamedRerouteUsage())) {
   atlas=/Game/Accessories/FORT_Backpacks/Backpack_M_MED_PrismParticle/Textures/CA_PrismParticle_Backpack
   curve=/Game/Accessories/FORT_Backpacks/Backpack_M_MED_PrismParticle/Textures/C_PrismParticle_Backpack
   group=PrismParticle
   parameterName=PP_PulseCurve
  }
  n55 = ScalarParameter() {defaultValue=1, group=PrismParticle, parameterName=PP_ReactivityToggle}
  NamedRerouteDeclaration(None=Saturate(None=Multiply(A=@n54.R, B=@n55))) {name=ReactivityGlow, nodeColor=<rgba 0.5469 0 1 1>}
  n56 = PixelNormalWS()
  n57 = Multiply(
   A=ComponentMask(None=Transform(None=@n56) {transformSourceType=TRANSFORMSOURCE_World, transformType=TRANSFORM_View}) {g=true, r=true}
   B=ScalarParameter() {group=PrismParticle, parameterName=PP_InteriorGlowDistortion}
  ) {constB=0.1}
  n58 = Multiply(
   A=SphereMask(
    A=Constant()
    B=Add(A=@n10.UVS, B=@n57)
    Radius=ScalarParameter() {defaultValue=10, group=PrismParticle, parameterName=PP_InternalGlowRadius}
   ) {attenuationRadius=0.5, hardnessPercent=0}
   B=ScalarParameter() {defaultValue=0.8, group=PrismParticle, parameterName=PP_InternalGlowIntensity}
  )
  NamedRerouteDeclaration(None=OneMinus(None=Saturate(None=@n58))) {name=2DGlowGradient, nodeColor=<rgba 1 0 0.9141 1>}
  n59 = Subtract(A=Time(), B=ScalarParameter() {group=PrismParticle, parameterName=PP_GameTimeFromBP_CCPM})
  n60 = ScalarParameter() {defaultValue=1.5, group=PrismParticle, parameterName=PP_ReactivityDuration_CCPM}
  n61 = Constant()
  n62 = Divide(A=@n59, B=@n60)
  NamedRerouteDeclaration(None=Saturate(None=If(A=@n59, B=@n60, "A > B"=@n61, "A == B"=@n61, "A < B"=@n62))) {name=Reactivity, nodeColor=<rgba 0.9688 0 1 1>}
  NamedRerouteUsage()
  Normalize(VectorInput=@n48)
  TextureSample(UVs=MaterialFunctionCall(X=@n22.R, Y=@n22.R) {materialFunction=&MakeFloat2}.Result) {samplerSource=SSM_Wrap_WorldGroupSettings, samplerType=SAMPLERTYPE_Masks, texture=&T_BlackMask}
 ]}
