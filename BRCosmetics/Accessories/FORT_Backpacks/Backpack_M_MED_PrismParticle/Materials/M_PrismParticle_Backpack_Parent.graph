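# Parent material for the PrismParticle backpack cosmetic, built on the shared 2023 character material functions.
# MF_CharacterParent_2023_PreFX feeds the backpack-specific MF_PrismParticle_Backpack function, whose result enters MF_CharacterParent_2023_PostFX alongside other PreFX outputs.
# It uses material attributes, world-space normals and an expression-driven shading model, with skeletal mesh, mesh deformer and Nanite usage enabled.
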Material /BRCosmetics/Accessories/FORT_Backpacks/Backpack_M_MED_PrismParticle/Materials/M_PrismParticle_Backpack_Parent
graph {
 properties {
  bTangentSpaceNormal=false
  bUseMaterialAttributes=true
  bUsedWithMeshDeformer=true
  bUsedWithNanite=true
  bUsedWithSkeletalMesh=true
  pixelDepthOffsetMode=PDOM_Legacy
  previewMesh=/BRCosmetics/Accessories/FORT_Backpacks/Backpack_M_MED_PrismParticle/Meshes/M_MED_PrismParticle_Pack
  shadingModel=MSM_FromMaterialExpression
  usedShadingModels=Default Lit | Subsurface | Clear Coat | Subsurface Profile | Cloth
 }
 nodes [
  n1 = MaterialFunctionCall() {materialFunction=&MF_CharacterParent_2023_PreFX}
  MaterialFunctionCall(
   Material=MaterialFunctionCall!(In=@n1.Result) {materialFunction=/Game/Accessories/FORT_Backpacks/Backpack_M_MED_PrismParticle/Materials/MF_PrismParticle_Backpack}.Result
   M=@n1.M
   Pre Skinned Local Position=@n1.M
   Distance Blend=@n1.M
   Fresnel=@n1.M
  ) {materialFunction=&MF_CharacterParent_2023_PostFX}
 ]
}
