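# Animation Blueprint for the Prism Particle back bling that makes its dyn_pack bone float and tumble procedurally.
# Each update, sine and cosine waves of real time set an ambient offset and rotation, which additive Modify Bone nodes apply after a small root offset.
# The update also writes dyn_pack's world position and basis axes, forward negated, into dynamic material pivot parameters, with object position turned off, so the shader follows the animated bone.
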AnimBlueprint /BRCosmetics/Accessories/FORT_Backpacks/Backpack_M_MED_PrismParticle/Meshes/M_MED_PrismParticle_Pack_AnimBP
target_skeleton=&/M_MED_PrismParticle_Pack_Skeleton
graphs [
 AnimGraph {nodes [
   n1 = "Transform (Modify) Bone - Bone: root"(ComponentPose=Local To Component(LocalPose=Input Pose()), Translation="0,1,0", Rotation="0,0,0") {
    translationMode=BMM_Additive
    translationSpace=BCS_BoneSpace
    alphaBoolBlend {blendOption=Linear}
    boneToModify {boneName=root}
    translation {y=1}
   }
   n2 = "Transform (Modify) Bone - Bone: dyn_pack"(ComponentPose=@n1, Alpha=1, Translation=Get Ambient Float() -> (Ambient Float:Vector), Rotation="0,0,0", Scale="1,1,1") {
    rotationMode=BMM_Additive
    rotationSpace=BCS_BoneSpace
    translationMode=BMM_Additive
    translationSpace=BCS_BoneSpace
    alphaBoolBlend {blendOption=Linear}
    boneToModify {boneName=dyn_pack}
   }
   n3 = "Transform (Modify) Bone - Bone: dyn_pack"(ComponentPose=@n2, Alpha=1, Translation="0,0,0", Rotation=Get Ambient Rotator() -> (Ambient Rotator:Rotator), Scale="1,1,1") {
    rotationMode=BMM_Additive
    rotationSpace=BCS_BoneSpace
    translationMode=BMM_Additive
    translationSpace=BCS_BoneSpace
    alphaBoolBlend {blendOption=Linear}
    boneToModify {boneName=dyn_pack}
   }
   Output Pose(Result=Component To Local(ComponentPose=@n3)) {layerGroup=None}
  ]}
 EventGraph {nodes [
   Event BlueprintUpdateAnimation()
   > n1 = Sequence!(execute) -> (then_0, then_1, then_2:Exec)
   n2 = Get Dynamic Material Instance(execute=@n1.then_2, SkeletalMeshComponent=GetPartSkeletalMeshComponent())
   > SetScalarParameterValue(self=@n2, ParameterName=PP_UseObjectPosition, Value=0)
   n3 = BreakTransform(InTransform=GetBoneTransform!(
     self:Skinned Mesh Component Object Reference=GetPartSkeletalMeshComponent()
     InBoneName:Name=dyn_pack
     TransformSpace:ERelativeTransformSpace Enum=RTS_World
    ) -> (ReturnValue:Transform))
   > SetVectorParameterValue(self=@n2, ParameterName=PP_PivotLocation, Value="To LinearColor (Vector)"!(InVec:Vector=@n3) -> (ReturnValue:Linear Color Structure))
   > SetVectorParameterValue(
    self=@n2
    ParameterName=PP_PivotBasisX
    Value="To LinearColor (Vector)"(InVec=GetUpVector!(InRot:Rotator=@n3.Rotation) -> (ReturnValue:Vector))
   )
   > SetVectorParameterValue(
    self=@n2
    ParameterName=PP_PivotBasisY
    Value="To LinearColor (Vector)"(InVec=GetRightVector!(InRot:Rotator=@n3.Rotation) -> (ReturnValue:Vector))
   )
   n4 = GetForwardVector!(InRot:Rotator=@n3.Rotation) -> (ReturnValue:Vector)
   > SetVectorParameterValue(self=@n2, ParameterName=PP_PivotBasisZ, Value="To LinearColor (Vector)"(InVec=vector * vector(A=@n4, B:float=-1) -> (ReturnValue=~)))
   n5 = GetRealTimeSeconds()
   n6 = "Sin (Radians)"(A=float * float(A=0.25, B=@n5))
   n7 = "Cos (Radians)"!(A:double="Sin (Radians)"(A=float * float(A=0.25, B=@n5))) -> (ReturnValue:double)
   n8 = "Sin (Radians)"(A=float * float(A=0.4, B=@n5))
   Set Ambient Rotator(
    execute=@n1.then_0
    Ambient Rotator:Rotator=MakeRotator(
     Roll=Remap Float(Input=@n8, OldRange=(X=0,Y=1), NewRange=(X=0,Y=360)) -> (Output=~)
     Pitch=float * float(A=@n7, B=180)
     Yaw=float * float(A=@n6, B=180)
    )
   ) -> (Output_Get:Rotator)
   n9 = GetRealTimeSeconds()
   n10 = "Sin (Radians)"(A=float * float(A=float * float(A=1.2, B=1.4), B=@n9))
   n11 = "Sin (Radians)"(A=float * float(A=float * float(A=1.1, B=1.1), B=@n9))
   n12 = "Sin (Radians)"(A=float * float(A=float * float(A=0.9, B=1.1), B=@n9))
   Set Ambient Float(
    execute=@n1.then_1
    Ambient Float:Vector=MakeVector(
     X=Remap Float(Input=@n10, OldRange=(X=-1,Y=1), NewRange=(X=1,Y=-6)) -> (Output=~)
     Y=float * float(A=@n11, B=0.15)
     Z=float * float(A=@n12, B=0.3)
    )
   ) -> (Output_Get:Vector)
   TryGetPawnOwner()
  ]}
]
variables [
 Ambient Float {type=Vector}
 Ambient Rotator {type=Rotator}
 Scale_0 {type=Vector}
 Scale_1 {type=Vector}
 Scale_2 {type=Vector}
]
